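; Phone booth hiding prop where players enter to change outfit, removed at play start if the playlist's enable row is off.
; Server entry allows loadout refresh, animates door and blinds, then applies the GameplayEffect granting the locker sequence ability.
; Blinds use custom primitive data, with replicated occupancy restoring state for late joiners.
; Overlap volumes drive the interaction widget, configured game-phase changes eject occupants, and Mesh Network events can live-disable booths.
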(event OnMeshNetEvent(FortEventResponder)
  (params (EventTag :type Gameplay%20Tag%20Structure) (bPersistentOnOff :type Boolean) (EventType :type EEventResponderEventType%20Enum :default "OneShotEvent") (EventParameters :type String))
  ;[0.1739 1 0.9303] Mesh Network support to live disable / re-enable these booths gracefully
  (CallFunction|LiveDisable bPersistentOnOff))

(event Custom|Live%20Disable
  (params (Disable :type Boolean))
  ;[0.1739 1 0.9303] Mesh Network support to live disable / re-enable these booths gracefully
  (bind _output_get (Variables|Default|SetDisabled Disable))
  (Variables|Interaction|SetbAllowInteract (not _output_get)) ; needed to prevent server interaction
  ;;[1 1 1] Visually update the booth
  (Networking|CanExecuteCosmeticEvents
    (:True
      (CallFunction|UpdateDisabledVisuals))))

(event Custom|Update%20Disabled%20Visuals
  ;[0.1739 1 0.9303] Mesh Network support to live disable / re-enable these booths gracefully
  ;;[1 1 1] TODO improve the disabled look
  (Rendering|SetVisibility (Variables|Default|GetUI_InteractionPoint_Upgrade) (not (Variables|Default|GetDisabled)) false))

(event EventEndPlay
  (params (EndPlayReason :type EEndPlayReason%20Enum :default "Destroyed"))
  ;[1 1 1] EndPlay
  (|CallParentFunction EndPlayReason)
  (bind _gamephasechangeeventtask (Variables|Transient|GetGamePhaseChangeEventTask)
    (:then
      (GameplayEventRouter|StopListeningtoStatefulEvent _gamephasechangeeventtask))
    (:else)))

(event EventBeginPlay
  (|CallParentFunction)
  ;[1 1 1] SetupUI
  (CallFunction|InitializeUI)
  ;[1 1 1] Reset Blinds Material
  (Variables|Default|SetBoothOcupied (Variables|Default|GetBoothOcupied))
  ;[1 1 1] Check If Enabled
  (Class|FortAsyncActionCurrentPlaylistReady|PlaylistDataReadyAsync
    (:OnReady
      (if (Class|AthenaFunctionLibrary|RowtoBool (Variables|Default|GetEnabledRow))
        (bind (_asynctaskproxy _eventrouter) (GameplayEventRouter|WaitForGlobalEventRouter 0)
          (:OnEventRouterAvailable
            (Networking|SwitchHasAuthority
              (:Authority
                (bind (_asynctaskproxy_2 _payload) (GameplayEventRouter|StartListeningToStatefulEvent _eventrouter "/Script/FortniteGame.FortGamePhaseUpdatedEvent")
                  (:then
                    (Variables|Transient|SetGamePhaseChangeEventTask _asynctaskproxy_2) ; As soon as there's more than one event to handle, \nswitch this to an array, or by then there may be a\nBetter way to handle GameplayEvents.
                  )
                  (:OnEventReceived
                    _)
                  (:OnSavedEventStateExists
                    _))
                (bind _payload_gamephase (Utilities|Struct|BreakFortGamePhaseUpdatedEvent _payload))
                (CallFunction|CheckforGamePhaseExit _payload_gamephase)))))
        (else
          (Gameplay|SilentDie false))))))

(event EventTeleport
  (params (Pawn :type Actor%20Object%20Reference)))

(event EventInteractEnter
  (bind _lastpawntointeract (Variables|Default|GetLastPawnToInteract))
  ;[1 1 1] Event -> Interact Enter - Call the events that open the Locker Widget when player enters the booth
  (Development|PrintString "Open the Locker")
  (Utilities|IsValid _lastpawntointeract
    (:Is%20Valid
      (Networking|SwitchHasAuthority
        (:Authority
          (CallFunction|PrepareforLoadoutRefresh _lastpawntointeract))))))

(event Custom|PrepareForLoadoutRefresh
  (params (InteractingPawn :type Fort%20Pawn%20Object%20Reference))
  ;[0.8634 1 0] SERVER: Prepare For Refresh - Call When they enter on the server
  (bind _asfort_player_controller_zone (Utilities|Casting|CastToFortPlayerControllerZone (Fortnite|GetFortPlayerControllerfromActor InteractingPawn)))
  (Cosmetics|AllowLoadoutRefresh _asfort_player_controller_zone)
  (CallFunction|DoorSwingAnim InteractingPawn)
  (CallFunction|BlindsAnim))

(event Custom|ActivateAbility
  (params (InteractingPawn :type Fort%20Pawn%20Object%20Reference))
  ;[1 1 1] Apply GE that gives GA that handles the whole phone booth sequence
  (Utilities|IsValid InteractingPawn
    (:Is%20Valid
      (GameplayEffects|FortApplyGameplayEffectToTarget (Ability|GetFortAbilitySystemComponent self) (Variables|Default|GetGE_EnterLocker) (Ability|GetFortAbilitySystemComponent InteractingPawn) 0 self self 0.0 (GameplayEffects|MakeFortEffectContext (Ability|GetFortAbilitySystemComponent InteractingPawn) 0 self self 0)))))

(event Custom|CheckForGamePhaseExit
  (params (GamePhase :type EAthenaGamePhase%20Enum%20%28by%20ref%29))
  ;[0 0.1435 0.15 0.5] Exit all Hiding players if configured to do so when this game phase is reached.
  (if (Utilities|Set|ContainsItem (Variables|BoothOptions|GetExitPlayersOnPhaseChanges) GamePhase)
    (for _array_element (Variables|Default|GetHidingPlayers)
      (bind _returnvalue (Ability|GetFortAbilitySystemComponent _array_element))
      (GameplayEffects|FortApplyGameplayEffectToTarget _returnvalue (Variables|BoothOptions|GetGE_ExitLocker) _returnvalue 0 self 0 0.0 (GameplayEffects|MakeEffectContext _returnvalue))
      (Default|DebugAbilityLogging true false "B_HidingProp_PapayaBooth.CheckingForGamePhaseExit.FortApplyGameplayEffectToTarget: GE_ExitPapayaBooth"))))

(event OnComponentBeginOverlap(SphereDetail)
  (params (OverlappedComponent :type Primitive%20Component%20Object%20Reference) (OtherActor :type Actor%20Object%20Reference) (OtherComp :type Primitive%20Component%20Object%20Reference) (OtherBodyIndex :type Integer) (bFromSweep :type Boolean) (SweepResult :type Hit%20Result%20Structure%20%28by%20ref%29))
  (CallFunction|CheckShouldDisplayUIV2 true true))

(event Custom|CheckShouldDisplayUIV2
  (params (ShowDetail :type Boolean) (ShowDistance :type Boolean))
  (bind _ui_interactionpoint_upgrade (Variables|Default|GetUI_InteractionPoint_Upgrade))
  (bind _returnvalue (UserInterface|GetUserWidgetObject _ui_interactionpoint_upgrade))
  ;[1 1 1] Toggle "Upgrade" Widget
  (Networking|CanExecuteCosmeticEvents
    (:True
      (if (not (Networking|IsServer))
        (Utilities|IsValid _returnvalue
          (:Is%20Valid
            (bind _asfort_player_controller (Utilities|Casting|CastToFortPlayerController (Widget|GetOwningPlayer _returnvalue)))
            (Utilities|IsValid (Game|Player|GetPlayerPawn _asfort_player_controller)
              (:Is%20Valid
                (CallFunction|SetUIVisibility ShowDetail ShowDistance)))))))))

(event OnComponentEndOverlap(SphereDetail)
  (params (OverlappedComponent :type Primitive%20Component%20Object%20Reference) (OtherActor :type Actor%20Object%20Reference) (OtherComp :type Primitive%20Component%20Object%20Reference) (OtherBodyIndex :type Integer))
  (CallFunction|CheckShouldDisplayUIV2 false true))

(event OnComponentBeginOverlap(CylinderDistance)
  (params (OverlappedComponent :type Primitive%20Component%20Object%20Reference) (OtherActor :type Actor%20Object%20Reference) (OtherComp :type Primitive%20Component%20Object%20Reference) (OtherBodyIndex :type Integer) (bFromSweep :type Boolean) (SweepResult :type Hit%20Result%20Structure%20%28by%20ref%29))
  (CallFunction|CheckShouldDisplayUIV2 false true))

(event Custom|InitializeUI
  (bind _ashud_interaction_point (Utilities|Casting|CastToHUD_InteractionPoint (UserInterface|GetUserWidgetObject (Variables|Default|GetUI_InteractionPoint_Upgrade))))
  (Class|HUDInteractionPoint|EventSetText _ashud_interaction_point "NSLOCTEXT(\"[242414D04BE65380D33E2DA9C2FF4B1D]\", \"360A2310488E3B9FF21C90B9AB1FAB6E\", \"Change\\r\\nOutfit\")"))

(event OnComponentEndOverlap(CylinderDistance)
  (params (OverlappedComponent :type Primitive%20Component%20Object%20Reference) (OtherActor :type Actor%20Object%20Reference) (OtherComp :type Primitive%20Component%20Object%20Reference) (OtherBodyIndex :type Integer))
  (CallFunction|CheckShouldDisplayUIV2 false))

(event Custom|SetUIVisibility
  (params (Show%20Detail :type Boolean) (Show%20Distance :type Boolean))
  (bind _ashud_interaction_point (Utilities|Casting|CastToHUD_InteractionPoint (UserInterface|GetUserWidgetObject (Variables|Default|GetUI_InteractionPoint_Upgrade))))
  (Class|HUDInteractionPoint|EventShowDetail _ashud_interaction_point Show%20Detail)
  (CallFunction|SetUIInteractionPointUpgradeHidden (not Show%20Distance)))

(event Custom|BlindsAnim
  (bind _staticmeshcomponent (Variables|Debug|GetStaticMeshComponent))
  (if (<= (Utilities|Array|Length (Variables|Default|GetHidingPlayers)) 1)
    ;[1 1 1] Animate Blinds Material
    (Components|Timeline|Stop (Variables|B_HidingProp_Papaya_Booth|GetTimeline_1))
    (Rendering|Material|SetCustomPrimitiveDataFloat _staticmeshcomponent 6 1.0)
    (bind (_direction _float) (Timeline|Timeline_0 PlayFromStart)
      (:Update
        (Rendering|Material|SetCustomPrimitiveDataFloat _staticmeshcomponent 5 _float))
      (:Finished
        (Variables|Default|SetBoothOcupied 1)))))

(event Custom|BlindsAnimExit
  ;[1 1 1] Animate Blinds Material
  (Components|Timeline|Stop (Variables|B_HidingProp_Papaya_Booth|GetTimeline_0))
  (bind (_direction _float) (Timeline|Timeline_1 PlayFromStart)
    (:Update
      (Rendering|Material|SetCustomPrimitiveDataFloat (Variables|Debug|GetStaticMeshComponent) 5 _float))
    (:Finished
      (Variables|Default|SetBoothOcupied 0))))

(event Custom|ChangingOutFitWobble
  (bind _returnvalue (Math|Random|RandomUnitVector))
  (bind _returnvalue_2 (Math|Random|RandomFloatinRange 6e+01 135.0))
  (bind _wobblelocationoffset (Variables|Default|GetWobbleLocationOffset))
  (bind _y (.y _wobblelocationoffset))
  (bind _x (.x _wobblelocationoffset))
  (bind _returnvalue_3 (Math|Vector|MakeVector _x _y _returnvalue_2))
  (bind _returnvalue_4 (* _returnvalue 0.2))
  (bind _self self)
  (bind _returnvalue_5 (Ability|GameplayCue|MakeGameplayCueParameters :NormalizedMagnitude 1e+02 :RawMagnitude 0.0 :Location _returnvalue_3 :Normal _returnvalue_4 :Instigator _self :GameplayEffectLevel 1 :AbilityLevel 1 :bReplicateLocationWhenUsingMinimalRepProxy false))
  ;[1 1 1] Wobble on Change Outfit (NOT USED)
  (GameplayCue|ExecuteGameplayCue%28Burst%29Local self (Variables|Default|GetGC_Wobble_0) _returnvalue_5)
  (Utilities|FlowControl|Delay 0.4)
  (GameplayCue|ExecuteGameplayCue%28Burst%29Local self (Variables|Default|GetGC_Wobble_0) _returnvalue_5)
  (Utilities|FlowControl|Delay 0.8)
  (GameplayCue|ExecuteGameplayCue%28Burst%29Local self (Variables|Default|GetGC_Wobble_0) _returnvalue_5))

(event Custom|DoorSwingAnim
  (params (InteractingPawn :type Fort%20Pawn%20Object%20Reference))
  ;[1 1 1] Door Swing Animation
  ;;[1 1 1 1 12] Fix for FORT-297163. Prevent server-replicated event from triggering 2nd door swing. Could still break if replication is delayed longer than it takes the door to close.
  (if (Components|Timeline|IsPlaying (Variables|B_HidingProp_Papaya_Booth|GetDoorSwing))
    ;;[0.09102 0.1428 0.15] Pops up the HUD right away for another player if he enters the phonebooth mid door animation
    (CallFunction|ActivateAbility InteractingPawn)
    (else
      (Variables|Default|SetDoorOpenPawn InteractingPawn)
      (Collision|SetCollisionResponsetoChannel (Variables|Default|GetSM_Telephone_Booth_DoorGlass) "ECC_Camera" "ECR_Ignore")
      (bind (_direction _float) (Timeline|DoorSwing PlayFromStart)
        (:Update
          (Transformation|SetRelativeRotation (Variables|Default|GetDoor) (Math|Rotator|Lerp(Rotator) "0,90.000000, 0" "0,-30.000000, 0" _float) false false))
        (:Finished
          (Components|Activation|Activate (Variables|Default|GetNS_PhantomBooth_DoorSlam) true)
          (CallFunction|ActivateAbility (Variables|Default|GetDoorOpenPawn)))))))

(event Custom|SecondaryExitKnockBack
  (params (PlayerPawn :type Fort%20Player%20Pawn%20Object%20Reference))
  (bind _outhits (Collision|MultiBoxTraceByChannel (+ (Transformation|GetWorldLocation (Variables|Default|GetHideLocation_ForwardVector)) (Variables|ObstructionTrace|GetObstructionTraceStartOffSet)) (+ (Transformation|GetWorldLocation (Variables|Default|GetHideLocation_ForwardVector)) (* (Math|Vector|SelectVector (+ (Transformation|GetForwardVector (Variables|Default|GetHideLocation_ForwardVector)) (Variables|ObstructionTrace|GetFixedEntraceObstructionTraceEndOffset)) (Math|Vector|MakeVector 0.0 0.0 1.0) (Variables|Interaction|GetFixedEntranceDirection)) (Variables|ObstructionTrace|GetObstructionTraceLength))) (Variables|ObstructionTrace|GetObstructionTraceExtents) (Math|Conversions|RotationFromXVector (Transformation|GetForwardVector (Variables|Default|GetHideLocation_ForwardVector))) "TraceTypeQuery5" true (Utilities|Array|MakeArray PlayerPawn)))
  (for _array_element _outhits
    (if (Variables|Actor|GetbCanBeDamaged (.HitActor (Collision|BreakHitResult _array_element)))
      (Collision|AddIgnoredActor PlayerPawn (.HitActor (Collision|BreakHitResult _array_element)) 0.25) ; Player Exiting Prop ignores Actors it's traced
      (bind _asfort_player_pawn (Utilities|Casting|CastToFortPlayerPawn (.HitActor (Collision|BreakHitResult _array_element))))
      (Character|LaunchCharacter _asfort_player_pawn (Math|Vector|RotateVectorAroundAxis (+ (* (Transformation|GetForwardVector (Variables|Default|GetHideLocation_ForwardVector)) 5e+02) "0, 0,750.000000") (select (Math|Random|RandomBool) (Math|Random|RandomFloatinRange 15.0 3e+01) (Math|Random|RandomFloatinRange -15.0 -3e+01)) "0, 0,1.000000") true true) ; Launch Player standing on prop in random direction
    )))

(fn BlueprintCanInteract
  (params (InteractingPawn :type Fort%20Pawn%20Object%20Reference) (InteractionBeingAttempted :type EInteractionBeingAttempted%20Enum :default "FirstInteraction") (InteractionType :type TInteractionType%20Enum :default "IT_NoInteraction"))
  (if (Class|AthenaFunctionLibrary|RowtoBool (Variables|Default|GetClientInteractiveRow))
    (return (|CallParentFunction InteractingPawn InteractionBeingAttempted InteractionType))
    (else
      (return false))))

(fn BlueprintGetInteractionString
  (params (InteractingPawn :type Fort%20Pawn%20Object%20Reference) (InteractionBeingAttempted :type EInteractionBeingAttempted%20Enum :default "FirstInteraction"))
  (|CallParentFunction InteractingPawn InteractionBeingAttempted)
  (return (Utilities|Text|FormatText "NSLOCTEXT(\"[242414D04BE65380D33E2DA9C2FF4B1D]\", \"09B0CC1A4E5AA8B24B25E7AD2A5A493D\", \"Change Appearance\")")))

(fn DebugAbilityLogging
  (params (BuildLogEnabled? :type Boolean :default true) (PIE%20PrintEnabled? :type Boolean) (Message :type String) (GameplayEventData :type Gameplay%20Event%20Data%20Structure%20%28by%20ref%29))
  (bind _instigator (Utilities|Struct|BreakGameplayEventData GameplayEventData))
  (bind _returnvalue (Utilities|GetDisplayName _instigator))
  (bind _returnvalue_2 (Utilities|String|Append Message _returnvalue))
  (Utilities|FlowControl|Sequence
    (:then_0
      (if BuildLogEnabled?
        (Fortnite|FortShippingLog self _returnvalue_2)))
    (:then_1
      (if PIE%20PrintEnabled?
        (Development|PrintString _returnvalue_2)))))

(fn GetFocusedSocketLocation
  (return (Transformation|GetWorldLocation (Variables|Default|GetInteractLocation))))

(fn OnRep_BoothOcupied
  (bind _boothocupied (Variables|Default|GetBoothOcupied))
  (bind _returnvalue (Math|Conversions|ToBoolean(Integer) _boothocupied))
  (bind _staticmeshcomponent (Variables|Debug|GetStaticMeshComponent))
  ;[1 1 1] Replicate the state of the blinds for clients joining in progress of out of relevancy
  (Rendering|Material|SetCustomPrimitiveDataFloat _staticmeshcomponent 5 (select _returnvalue 0.0 1.0))
  (Rendering|Material|SetCustomPrimitiveDataFloat _staticmeshcomponent 6 (select _returnvalue 0.0 1.0))
  (Development|PrintString (Utilities|String|ToString(Boolean) _returnvalue)))

(fn SetUI_InteractionPoint_UpgradeHidden
  (params (InHidden :type Boolean))
  (if (!= InHidden (Variables|/Script/Engine.SceneComponent|GetbHiddenInGame (Variables|Default|GetUI_InteractionPoint_Upgrade)))
    (Development|SetHiddeninGame (Variables|Default|GetUI_InteractionPoint_Upgrade) (Variables|GetInHidden) false)
    (if (not (Variables|GetInHidden))
      (bind _ashud_interaction_point (Utilities|Casting|CastToHUD_InteractionPoint (UserInterface|GetUserWidgetObject (Variables|Default|GetUI_InteractionPoint_Upgrade))))
      (Class|HUDInteractionPoint|EventIntro _ashud_interaction_point))
    (return true)
    (else
      (return false))))

(fn ConstructionScript
  (|CallParentFunction)
  (if (Fortnite|IsMobilePlatform)
    (Components|DestroyComponent (Variables|Default|GetSM_Telephone_Booth_GlassWindows))))
