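; Portapotty hiding prop that swings its door on use and swaps players on the Deadpool (Donut) week 7 quest into his costume.
; Clients async preload costume parts at BeginPlay if that quest is active.
; When the server-triggered door timeline finishes, authority grants the costume gadget, sends the quest stat event and calls the sound-flag setter on this prop or its target teleporter.
; The flagged local client hears a completion sound on hiding changes.
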(event EventBeginPlay
  (bind _returnvalue (Fortnite|GetFirstLocalFortPlayerController true))
  (bind _returnvalue_2 (Fortnite|GetFirstLocalFortPlayerController true))
  (bind _donutquest (Variables|Default|GetDonutQuest))
  ;[1 1 1] Begin Play - Check to see if each client is on the Deadpool quest. If so, async load the Deadpool costume assets.
  (|CallParentFunction)
  (Networking|SwitchHasAuthority
    (:Remote
      (Utilities|IsValid _returnvalue
        (:Is%20Valid
          (Quests|CurrentQuestsReadyAsync (Quests|GetQuestManager _returnvalue "Invalid") (Utilities|Array|MakeArray (Variables|Default|GetDonutQuest))
            (:OnCurrentQuestsReady
              (bind _questvalid (Quest|DoesPlayerHaveQuest _returnvalue_2 _donutquest "Quest_S12_Donut_W7_B"))
              (bind (_playercontrollerout _completedquestobjective) (Quest|HasPlayerCompletedQuestObjective _returnvalue_2 _donutquest "Quest_S12_Donut_W7_B"))
              (if (and _questvalid (not _completedquestobjective))
                (for _array_element (Variables|Default|GetDonutCharacterPartsToApply)
                  (Utilities|AsyncLoadAsset (Utilities|ResolveSoftReference _array_element)))))))) ; This to avoid errors when the controller isn't valid
    )))

(event EventHidingPlayerCountChanged
  (|CallParentFunction)
  (Utilities|FlowControl|Sequence
    (:then_0
      (Networking|SwitchHasAuthority
        (:Remote
          (Variables|Default|GetTargetTeleporter
            (:then)
            (:else
              (if (Variables|Default|GetPlayDonutSound)
                (CallFunction|PlayDonutSoundforPlayer)))))))
    (:then_1
      (if (GameplayTags|HasMatchingGameplayTag (Variables|Default|GetLastPawnToHide) (Variables|Default|GetTeleportingTag))
        _
        (elif (Networking|IsServer)
          (CallFunction|DoorSwingAnim))))))

(event EventInteractEnter
  (Components|Activation|Activate (Variables|Default|GetP_HidingProp_Portapotty) true)
  (if (Networking|IsServer)
    (CallFunction|DoorSwingAnim)))

(event Custom|DoorSwingAnim
  (bind (_direction _float) (Timeline|DoorSwing PlayFromStart)
    (:Update
      (Transformation|SetRelativeRotation (Variables|Default|GetConstr_Portapotty_Searched_Door) (Math|Rotator|Lerp(Rotator) "0, 0, 0" "0,140.000000, 0" _float) false false))
    (:Finished
      (CallFunction|TrytoUpdateDeadpoolQuest))))

(event Custom|TryToUpdate_DeadpoolQuest
  (bind _lastpawntohide (Variables|Default|GetLastPawnToHide))
  (bind _returnvalue (Fortnite|GetFortPlayerControllerfromActor _lastpawntohide))
  (bind _donutquest (Variables|Default|GetDonutQuest))
  ;[1 1 1] Swap to Deadpool costume if player is on the Deadpool quest
  (Networking|SwitchHasAuthority
    (:Authority
      (bind _asfort_player_pawn (Utilities|Casting|CastToFortPlayerPawn (Variables|Default|GetLastPawnToHide)))
      (bind _questvalid (Quest|DoesPlayerHaveQuest (Fortnite|GetFortPlayerControllerfromActor _asfort_player_pawn) _donutquest "Quest_S12_Donut_W7_B"))
      (bind (_playercontrollerout _completedquestobjective) (Quest|HasPlayerCompletedQuestObjective (Fortnite|GetFortPlayerControllerfromActor _asfort_player_pawn) _donutquest "Quest_S12_Donut_W7_B"))
      (if (and _questvalid (not _completedquestobjective))
        (Inventory|GiveItemtoInventoryOwner _returnvalue (Variables|Default|GetDonutGadget))
        (Quest|SendComplexCustomStatEvent (Quests|GetQuestManager _returnvalue "Invalid") 0 "(GameplayTags=)" "(GameplayTags=((TagName=\"Athena.Quests.Donut.BecomeDonut\")))" 1)
        (bind _targetteleporter (Variables|Default|GetTargetTeleporter)
          (:then
            (bind _asb_hiding_prop_portapotty (Utilities|Casting|CastToB_HidingProp_Portapotty _targetteleporter))
            (CallFunction|SetPlayDonutSoundBool _asb_hiding_prop_portapotty _returnvalue))
          (:else
            (CallFunction|SetPlayDonutSoundBool _returnvalue)))))))

(event EventStopHiding
  (params (Pawn :type Fort%20Pawn%20Object%20Reference))
  (|CallParentFunction Pawn)
  (CallFunction|PlayDonutSoundMulticast))

(event Custom|PlayDonutSoundForPlayer
  (Audio|PlaySound2D (Variables|Default|GetDonutChangeCompleteSound))
  (Variables|Default|SetPlayDonutSound false))

(event Custom|PlayDonutSoundMulticast
  (Networking|SwitchHasAuthority
    (:Remote
      (if (Variables|Default|GetPlayDonutSound)
        (Audio|PlaySound2D (Variables|Default|GetDonutChangeCompleteSound))
        (Variables|Default|SetPlayDonutSound false)))))

(event Custom|SetPlayDonutSoundBool
  (params (PlayerController :type Fort%20Player%20Controller%20Object%20Reference))
  (Networking|SwitchHasAuthority
    (:Remote
      (if (Utilities|Equal(Object) PlayerController (Fortnite|GetFirstLocalFortPlayerController true))
        (Variables|Default|SetPlayDonutSound true)))))

(fn ConstructionScript
  (|CallParentFunction))
