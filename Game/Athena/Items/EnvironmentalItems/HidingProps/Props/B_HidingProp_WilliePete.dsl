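; Whirlpool hiding prop (WilliePete) that launches players leaving it, plus actors overlapping its center.
; Once the playlist is ready, the server snaps it to ocean height, then removes it based on water level or spawns its water body, with RepNotifies syncing location and visibility.
; Players receive a gameplay effect, event and quest stat; vehicles, AI and their tethered followers get direct launches.
; Eligible non-pawns get their velocity redirected along an exit vector.
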(event EventBeginPlay
  (bind (_returnvalue_x _returnvalue_y) (Math|Vector|BreakVector (Transformation|GetActorLocation)))
  (bind (_returnvalue_x_2 _returnvalue_y_2 _returnvalue_z) (Math|Vector|BreakVector (Transformation|GetActorLocation)))
  ;[0.6934 1 0.5064] On begin play, set height to total ocean height, set the rotation of the FVector reference component for when projectiles exit, and delete if below the desired water level. Instance editable and should be set as needed when placed.
  (Networking|SwitchHasAuthority
    (:Authority
      (Class|FortAsyncActionCurrentPlaylistReady|PlaylistDataReadyAsync
        (:OnReady
          (if (Class|AthenaFunctionLibrary|RowtoBool (Variables|Default|GetEnabledValue))
            (Variables|Default|SetAdjustedLocation (Math|Vector|MakeVector _returnvalue_x _returnvalue_y (Water|GetOceanTotalHeight (WorldSubsystems|GetWaterSubsystem))))
            ;;[1 1 1] Disabling for now since some maps in Apollo_Terrain seem to have this set!!! Bad.
            (Networking|ForceNetUpdate)
            (if (>= (Math|Conversions|ToFloat(Integer) (Variables|Default|GetwaterLevel)) (/ (- _returnvalue_z (Water|GetOceanBaseHeight (WorldSubsystems|GetWaterSubsystem))) 384.0))
              (Variables|Default|SetbSetSilentDie true)
              (Gameplay|SilentDie false)
              (else
                (CallFunction|CompleteSetup)))
            (else
              (Actor|SetLifeSpan (Math|Random|RandomFloatinRange 0.01 5.0)))))))))

(event Custom|Complete%20Setup
  (bind (_returnvalue_location _returnvalue_rotation _returnvalue_scale) (Math|Transform|BreakTransform (Transformation|GetActorTransform)))
  ;[0.6934 1 0.5064] On begin play, set height to total ocean height, set the rotation of the FVector reference component for when projectiles exit, and delete if below the desired water level. Instance editable and should be set as needed when placed.
  (Transformation|SetWorldRotation (Variables|Default|GetProjectileLocation_ForwardVector) (Variables|Default|GetProjectileExitFVectorRotation) false false)
  (CallFunction|HelperLocationLogger "Spawning child")
  (Game|SpawnActorfromClass (Variables|Default|GetSpawnedWaterBody) (Math|Transform|MakeTransform (Variables|Default|GetAdjustedLocation) _returnvalue_rotation _returnvalue_scale) "AlwaysSpawn" "SelectDefaultAtRuntime" (Game|GetGameState) 0)
  (Variables|Default|SetbSetSpawnedWaterBody true))

(event EventStopHiding
  (params (Pawn :type Fort%20Pawn%20Object%20Reference))
  ;[1 0.6149 0.4532] Launch actors that overlap center
  (|CallParentFunction Pawn)
  (bind _asfort_player_pawn_athena (Utilities|Casting|CastToFortPlayerPawnAthena Pawn))
  (Variables|Default|SetExitingPawn _asfort_player_pawn_athena)
  (Variables|Default|SetskipAnimForLaunch true)
  (Variables|Default|SetLaunchPawn _asfort_player_pawn_athena)
  (CallFunction|ExecuteGameplayCue (Variables|Default|GetGC_Enter) (Ability|GameplayCue|MakeGameplayCueParameters 0.0))
  (Audio|PlaySoundatLocation (Variables|Default|GetLaunch%20Sound) (Transformation|GetActorLocation))
  (Networking|SwitchHasAuthority
    (:Authority
      (bind _returnvalue (GameplayEffects|FortApplyGameplayEffectToTarget :self (Ability|GetFortAbilitySystemComponent self) :GameplayEffectClass "/Game/Athena/Items/EnvironmentalItems/HidingProps/Props/GE_WilliePete_PlayerLaunch_Granted.GE_WilliePete_PlayerLaunch_Granted_C" :TargetOfEffect (Ability|GetFortAbilitySystemComponent (Variables|Default|GetLaunchPawn)) :OptionalSourceObject self :Level 0.0))
      (Variables|Default|SetLaunchGrantedEffectHandle _returnvalue)
      (Ability|SendGameplayEventtoActor (Variables|Default|GetLaunchPawn) "(TagName=\"Event.Athena.WilliePete.Enter\")" (Utilities|Struct|MakeGameplayEventData :Instigator (Variables|Default|GetLaunchPawn) :Target self :OptionalObject (Variables|Default|GetExitingPawn) :ContextHandle "()" :InstigatorTags "(GameplayTags=)" :TargetTags "(GameplayTags=)" :EventMagnitude 0.0 :TargetData "()"))
      ;;[1 1 1] Quest Update
      (Quest|SendComplexCustomStatEvent (Quests|GetQuestManager (Fortnite|GetFortPlayerControllerfromActor (Variables|Default|GetLaunchPawn)) "Invalid") 0 (Variables|Default|GetT_Quest) (GameplayTags|GetOwnedGameplayTags :!class /Script/GameplayTags.GameplayTagAssetInterface) 1)
      (Variables|Default|SetLaunchPawn 0))))

(event OnComponentBeginOverlap(overlapCylinder)
  (params (OverlappedComponent :type Primitive%20Component%20Object%20Reference) (OtherActor :type Actor%20Object%20Reference) (OtherComp :type Primitive%20Component%20Object%20Reference) (OtherBodyIndex :type Integer) (bFromSweep :type Boolean) (SweepResult :type Hit%20Result%20Structure%20%28by%20ref%29))
  (bind _launchheightvalue (Variables|Default|GetLaunchHeightValue))
  (bind _returnvalue (ScalableFloat|GetValueatLevel _launchheightvalue))
  (bind _returnvalue_2 (* _returnvalue 0.6))
  ;[1 0.6149 0.4532] Launch actors that overlap center
  (Utilities|IsValid (Variables|Default|GetExitingPawn)
    (:Is%20Not%20Valid
      (Variables|Default|SetSphereOverlapResult SweepResult)
      (bind _returnvalue_3 (GameplayTags|HasMatchingGameplayTag(Message) OtherActor "(TagName=\"Item.AppleSun\")"))
      (if _returnvalue_3
        (Actor|DestroyActor OtherActor)
        (else
          (bind _asfort_player_pawn_athena (Utilities|Casting|CastToFortPlayerPawnAthena OtherActor)
            (:then
              (if (not (Class|FortPlayerPawn|IsTethered _asfort_player_pawn_athena))
                (Utilities|IsinVehicle _asfort_player_pawn_athena
                  (:false
                    (if (not (or (GameplayTags|HasMatchingGameplayTag (Ability|GetAbilitySystemComponent _asfort_player_pawn_athena) "(TagName=\"Gameplay.Action.Player.Launched\")") (Utilities|Equal(Object) (Variables|Default|GetExitingPawn) _asfort_player_pawn_athena)))
                      (Variables|Default|SetLaunchPawn _asfort_player_pawn_athena)
                      (CallFunction|ExecuteGameplayCue (Variables|Default|GetGC_Enter) (Ability|GameplayCue|MakeGameplayCueParameters 0.0))
                      (Audio|PlaySoundatLocation (Variables|Default|GetLaunch%20Sound) (Transformation|GetActorLocation))
                      (Networking|SwitchHasAuthority
                        (:Authority
                          (bind _returnvalue_4 (GameplayEffects|FortApplyGameplayEffectToTarget :self (Ability|GetFortAbilitySystemComponent self) :GameplayEffectClass "/Game/Athena/Items/EnvironmentalItems/HidingProps/Props/GE_WilliePete_PlayerLaunch_Granted.GE_WilliePete_PlayerLaunch_Granted_C" :TargetOfEffect (Ability|GetFortAbilitySystemComponent (Variables|Default|GetLaunchPawn)) :OptionalSourceObject self :Level 0.0))
                          (Variables|Default|SetLaunchGrantedEffectHandle _returnvalue_4)
                          (Ability|SendGameplayEventtoActor (Variables|Default|GetLaunchPawn) "(TagName=\"Event.Athena.WilliePete.Enter\")" (Utilities|Struct|MakeGameplayEventData :Instigator (Variables|Default|GetLaunchPawn) :Target self :OptionalObject (Variables|Default|GetExitingPawn) :ContextHandle "()" :InstigatorTags "(GameplayTags=)" :TargetTags "(GameplayTags=)" :EventMagnitude 0.0 :TargetData "()"))
                          ;;[1 1 1] Quest Update
                          (Quest|SendComplexCustomStatEvent (Quests|GetQuestManager (Fortnite|GetFortPlayerControllerfromActor (Variables|Default|GetLaunchPawn)) "Invalid") 0 (Variables|Default|GetT_Quest) (GameplayTags|GetOwnedGameplayTags :!class /Script/GameplayTags.GameplayTagAssetInterface) 1)
                          (Variables|Default|SetLaunchPawn 0)))
                      (else
                        (Variables|Default|SetLaunchPawn 0)))))))
            (:CastFailed
              (bind _asfort_vehicle_interface (Utilities|Casting|CastToFortVehicleInterface OtherActor)
                (:then
                  (Vehicle|SetIgnoreNextFallingDamage _asfort_vehicle_interface true)
                  (Utilities|FlowControl|Delay 0.5)
                  (Audio|BroadcastAbilityImpactatLocation 0 (Vehicle|GetActor _asfort_vehicle_interface) (+ (Vehicle|GetVehicleVelocity _asfort_vehicle_interface) (+ (* (Vehicle|GetVehicleForwardVector _asfort_vehicle_interface) (/ (ScalableFloat|GetValueatLevel (Variables|Default|GetLaunchHeightValue)) 5.0)) (* (Transformation|GetActorUpVector) (ScalableFloat|GetValueatLevel (Variables|Default|GetLaunchHeightValue))))) 8e+02))
                (:CastFailed
                  (bind _asfort_aipawn (Utilities|Casting|CastToFortAIPawn OtherActor)
                    (:then
                      (bind (_returnvalue_x _returnvalue_y) (Math|Vector|BreakVector (Transformation|GetActorForwardVector _asfort_aipawn)))
                      (Character|LaunchCharacter _asfort_aipawn (Math|Vector|MakeVector (* _returnvalue_x 1e+03) _returnvalue_y _returnvalue_2) false true)
                      (bind _tetheredfollower (Variables|/Script/FortniteGame.FortAIPawn|GetTetheredFollower _asfort_aipawn)
                        (:then
                          (Pawn|Character|LaunchCharacterJump _tetheredfollower (* (Math|Vector|MakeVector (* _returnvalue_x 1e+03) _returnvalue_y _returnvalue_2) 0.95) false true true true)
                          ;;[1 1 1] Quest Update
                          (FortPawn|SetGravityMultiplier _tetheredfollower 0.9)
                          (EventDispatchers|BindEventtoOnPawnLanded _tetheredfollower (delegate Custom|ResetGravity))
                          (Quest|SendComplexCustomStatEvent (Quests|GetQuestManager (Fortnite|GetFortPlayerControllerfromActor _tetheredfollower) "Invalid") 0 (Variables|Default|GetT_Quest) (GameplayTags|GetOwnedGameplayTags :!class /Script/GameplayTags.GameplayTagAssetInterface) 1))
                        (:else
                          _)))
                    (:CastFailed
                      (if (Utilities|IsValid OtherActor :!class /Script/Engine.KismetSystemLibrary)
                        (CallFunction|NonPawnTeleport OtherActor))))))
              (CallFunction|ExecuteGameplayCue (Variables|Default|GetGC_Exit) (Ability|GameplayCue|MakeGameplayCueParameters :NormalizedMagnitude 0.0 :RawMagnitude 0.0 :Location (Transformation|GetActorLocation) :Normal "0,0,0" :GameplayEffectLevel 1 :AbilityLevel 1 :bReplicateLocationWhenUsingMinimalRepProxy false)))))))))

(event Custom|Remove%20GE
  ;[1 0.6149 0.4532] Launch actors that overlap center
  ;;[1 1 1] Remove screen fx
  (GameplayEffects|RemoveActiveGameplayEffect (Ability|GetFortAbilitySystemComponent (Variables|Default|GetLaunchPawn)) (Variables|Default|GetLaunchGrantedEffectHandle) -1))

(event Custom|ResetGravity
  (params (Hit :type Hit%20Result%20Structure%20%28by%20ref%29))
  ;[1 0.6149 0.4532] Launch actors that overlap center
  ;;[1 1 1] Quest Update
  (FortPawn|SetGravityMultiplier (stale OnComponentBeginOverlap(overlapCylinder) _tetheredfollower) 1.0))

(event Custom|Non%20Pawn%20Teleport
  (params (TeleportingActor :type Actor%20Object%20Reference))
  (bind (_returnvalue_x _returnvalue_y _returnvalue_z) (Math|Vector|BreakVector (Transformation|GetForwardVector (Variables|Default|GetProjectileLocation_ForwardVector))))
  (bind _returnvalue (Actor|GetComponentbyClass TeleportingActor "/Script/Engine.MovementComponent"))
  (bind _returnvalue_2 (Actor|GetComponentbyClass TeleportingActor "/Script/Engine.ProjectileMovementComponent"))
  (bind (_returnvalue_x_2 _returnvalue_y_2 _returnvalue_z_2) (Math|Vector|BreakVector (Transformation|GetForwardVector (Variables|Default|GetProjectileLocation_ForwardVector))))
  ;[0.069 0.3722 1] Teleporting non-pawn actors.
  ;;[0.1418 0.6771 0.1093] Filter for valid actors that have projectile movement components, aren't excessively fast, and don't have 0 gravity scale. Disable movement.
  (Variables|Default|SetTeleportingNonPawn TeleportingActor)
  (bind _returnvalue_3 (Projectile|GetFortProjectileMovementComponent(Message) TeleportingActor))
  (Utilities|IsValid _returnvalue_3
    (:Is%20Valid
      (if (and (<= (Variables|/Script/Engine.ProjectileMovementComponent|GetInitialSpeed _returnvalue_3) (Variables|Default|GetProjectileSpeedCeiling)) (> (Variables|/Script/Engine.ProjectileMovementComponent|GetProjectileGravityScale _returnvalue_3) 0.0))
        (Variables|Default|SetOverlappedFortProjectileMovementComponent _returnvalue_3)))
    (:Is%20Not%20Valid
      (Utilities|IsValid _returnvalue_2
        (:Is%20Valid
          (if (and (<= (Variables|/Script/Engine.ProjectileMovementComponent|GetInitialSpeed _returnvalue_2) (Variables|Default|GetProjectileSpeedCeiling)) (> (Variables|/Script/Engine.ProjectileMovementComponent|GetProjectileGravityScale _returnvalue_2) 0.0))
            (Variables|Default|SetOverlappedStandardProjectileMovementComponent _returnvalue_2)))
        (:Is%20Not%20Valid
          (Utilities|IsValid _returnvalue
            (:Is%20Valid
              (Variables|Default|SetOverlappedMovementComponent _returnvalue))
            (:Is%20Not%20Valid
              (Variables|Default|SetTeleportingNonPawn 0)))))))
  ;;[0.9364 1 0.1667] Set teleporting boolean, cancel suction timeline, set variables to clear if teleported non-pawn dies, disable non-pawn collision and visibility, and tell target teleporter it's busy.
  (Game|BindEventtoOnDestroyed (Variables|Default|GetTeleportingNonPawn) (delegate Custom|Non%20Pawn%20Actor%20Destroyed))
  (Networking|SwitchHasAuthority
    (:Authority
      (Variables|Default|GetTeleportingNonPawn
        (:then
          ;;[1 0.3734 0] Checks if actor has Fort Projectile Movement component or Projectile Movement component override exit speed
          (bind _overlappedfortprojectilemovementcomponent (Variables|Default|GetOverlappedFortProjectileMovementComponent)
            (:then
              (Components|Activation|Activate _overlappedfortprojectilemovementcomponent false)
              (Projectile|SetVelocity (Variables|Default|GetOverlappedFortProjectileMovementComponent) (* (Math|Vector|MakeVector (* _returnvalue_x (Math|Random|RandomFloatinRange -1.0 2.0)) (* _returnvalue_y (Math|Random|RandomFloatinRange -1.0 2.0)) (* _returnvalue_z (Math|Random|RandomFloatinRange 1.0 2.0))) (Variables|Default|GetProjectileSpeedMult))))
            (:else
              (Variables|Default|GetOverlappedStandardProjectileMovementComponent
                (:then
                  (Game|Components|ProjectileMovement|SetVelocityinLocalSpace (Variables|Default|GetOverlappedStandardProjectileMovementComponent) (* (Math|Vector|MakeVector (* _returnvalue_x_2 (Math|Random|RandomFloatinRange -1.0 2.0)) (* _returnvalue_y_2 (Math|Random|RandomFloatinRange -1.0 2.0)) _returnvalue_z_2) (* (Variables|Default|GetProjectileSpeedMult) (Variables|Default|GetPickupSpeedMult)))))
                (:else
                  (bind _overlappedmovementcomponent (Variables|Default|GetOverlappedMovementComponent)
                    (:then
                      (Variables|/Script/Engine.MovementComponent|SetVelocity (* (Transformation|GetForwardVector (Variables|Default|GetProjectileLocation_ForwardVector)) (Variables|Default|GetProjectileSpeedMult)) _overlappedmovementcomponent))
                    (:else
                      (Variables|Default|SetTeleportingNonPawn 0 :!label _label)))))))
          ;;[1 0.1179 0] clear variable and play fx
          (Variables|Default|SetTeleportingNonPawn 0)
          (CallFunction|ExecuteGameplayCue (Variables|Default|GetGC_Exit) (Ability|GameplayCue|MakeGameplayCueParameters 0.0)))
        (:else
          _))
      (goto _label))))

(event Custom|Non%20Pawn%20Actor%20Destroyed
  (params (DestroyedActor :type Actor%20Object%20Reference))
  ;[0.069 0.3722 1] Teleporting non-pawn actors.
  ;;[0.9364 1 0.1667] Set teleporting boolean, cancel suction timeline, set variables to clear if teleported non-pawn dies, disable non-pawn collision and visibility, and tell target teleporter it's busy.
  (Variables|Default|SetTeleportingNonPawn 0))

(fn HelperLocationLogger
  (params (InString :type String))
  (Fortnite|FortShippingLog self (Utilities|String|Append "[Whirlpool Setup] " InString " [" (Utilities|String|ToString(Vector) (Transformation|GetActorLocation)) "]") true))

(fn OnRep_AdjustedLocation
  (CallFunction|HelperLocationLogger "Start Location")
  (Fortnite|FortShippingLog self (Utilities|String|Append "[Whirlpool Setup] Water Level [" (Utilities|String|ToString(Float) (/ (Water|GetOceanFloodHeight (WorldSubsystems|GetWaterSubsystem)) 384.0)) "] Ocean Height [" (Utilities|String|ToString(Float) (Water|GetOceanTotalHeight (WorldSubsystems|GetWaterSubsystem))) "]") true)
  (Transformation|SetMobility (Variables|Debug|GetStaticMeshComponent) "Movable")
  (bind (_sweephitresult _returnvalue) (Transformation|SetActorLocation (Variables|Default|GetAdjustedLocation)))
  (Transformation|SetMobility (Variables|Debug|GetStaticMeshComponent) "Static")
  (if _returnvalue
    (CallFunction|HelperLocationLogger "Adjusted Water Height Location")
    (else
      (CallFunction|HelperLocationLogger (Utilities|String|Append "Failed Adjusting to [" (Utilities|String|ToString(Vector) (Variables|Default|GetAdjustedLocation)) "] Current Location")))))

(fn OnRep_bSetSilentDie
  (Networking|SwitchHasAuthority
    (:Remote
      (Development|SetHiddeninGame (Variables|Debug|GetStaticMeshComponent) true false))))

(fn OnRep_bSetSpawnedWaterBody
  (Networking|SwitchHasAuthority
    (:Remote
      (Development|SetHiddeninGame (Variables|Default|GetS_Whirlpool_01) false false))))

(fn TeleportExitSpeed
  (params (ExitFVector :type Vector) (EnterSpeed :type Float%20%28double-precision%29) (OverrideSpeed :type Float%20%28double-precision%29) (MinimumSpeed :type Float%20%28double-precision%29) (ForceSpeedOverride :type Boolean))
  (return (* ExitFVector (select ForceSpeedOverride OverrideSpeed (select (< EnterSpeed MinimumSpeed) OverrideSpeed EnterSpeed)))))

(fn ConstructionScript
  (|CallParentFunction))
