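; Blueprint for a water body child used by the WilliePete hiding prop, going by its name and folder; the graph holds only spawn logging.
; On BeginPlay it writes a Fortnite shipping log line with its display name, whether it runs on server or client by authority, and its world location.
; The overlap and tick events and the construction script are empty, so no other behavior is defined here.
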(event EventBeginPlay
  (Fortnite|FortShippingLog self (Utilities|String|Append (Utilities|GetDisplayName self) " Spawning [" (Utilities|String|SelectString "Server" "Client" (Networking|HasAuthority)) "] at " (Utilities|String|ToString(Vector) (Transformation|GetActorLocation))) true))

(event Collision|EventActorBeginOverlap
  (params (OtherActor :type Actor%20Object%20Reference)))

(event EventTick
  (params (DeltaSeconds :type Float%20%28single-precision%29)))

(fn ConstructionScript)
