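; Gameplay ability that launches a player upward from the WilliePete hiding prop, then starts skydiving.
; Activation finds the launcher through the granting effect context, drops held objects, or launches exiting/skip-animation pawns, otherwise playing an exit montage unless DBNO.
; A capsule trace and hit events destroy blocking structures with a borrowed sword-slam effect.
; Jump apex or a fallback timer triggers skydiving for non-DBNO pawns and removes the launch-tagged effect.
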(event Ability|EventFailedToActivatePassiveAbility
  (params (ActorInfo :type Gameplay%20Ability%20Actor%20Info%20Structure))
  (|CallParentFunction))

(event Ability|EventActivateAbility
  (bind _owningpawn (Variables|Default|GetOwningPawn))
  (bind _asb_hiding_prop_willie_pete (Utilities|Casting|CastToB_HidingProp_WilliePete (Ability|EffectContext|GetInstigatorActor (Ability|GetGrantedbyEffectContext)))
    (:then
      (Variables|Default|SetWilliePeteLauncher _asb_hiding_prop_willie_pete)
      (bind _asfort_player_pawn_athena (Utilities|Casting|CastToFortPlayerPawnAthena (Ability|GetAvatarActorfromActorInfo))
        (:then
          (bind _output_get (Variables|Default|SetOwningPawn _asfort_player_pawn_athena))
          (Utilities|IsValid _output_get
            (:Is%20Valid
              (Ability|CommitAbility)
              (Ability|SendGameplayEventtoActor (Variables|Default|GetOwningPawn) "(TagName=\"Event.Athena.Item.FloppingRabbit.ForceEnd\")")
              (Pawn|Components|CharacterMovement|SetMovementMode (Variables|/Script/Engine.Character|GetCharacterMovement (Variables|Default|GetOwningPawn)) "MOVE_Falling" 0)
              ;[0.3932 1 0.4388] If holding something drop it. If DBNO or exiting teleporter then skip animations
              (if (Class|FortPawn|GetIsHoldingObject _owningpawn)
                (Class|FortPawn|EquipBestWeapon (Variables|Default|GetOwningPawn))
                (elif (or (Utilities|Equal(Object) _owningpawn (Variables|/Game/Athena/Items/EnvironmentalItems/HidingProps/Props/B_HidingProp_WilliePete.B_HidingProp_WilliePete_C|GetExitingPawn (Variables|Default|GetWilliePeteLauncher))) (Variables|/Game/Athena/Items/EnvironmentalItems/HidingProps/Props/B_HidingProp_WilliePete.B_HidingProp_WilliePete_C|GetskipAnimForLaunch (Variables|Default|GetWilliePeteLauncher)))
                  ;[1 0.9188 0.3083] Spin Montage
                  ;[1 0.7779 0] Break blocking structures and launch player
                  (CallFunction|StructureTraceLoop)
                  (Analytics|FireAnalyticEventEnvironmentalItemUsed (Fortnite|GetFortPlayerControllerfromActor (Variables|Default|GetOwningPawn)) (Variables|Default|GetWilliePeteLauncher) 0.0 "PlayerExit")
                  (Pawn|Character|LaunchCharacterJump (Variables|Default|GetOwningPawn) (Math|Vector|MakeVector 0.0 0.0 (ScalableFloat|GetValueatLevel (Variables|/Game/Athena/Items/EnvironmentalItems/HidingProps/Props/B_HidingProp_WilliePete.B_HidingProp_WilliePete_C|GetLaunchHeightValue (Variables|Default|GetWilliePeteLauncher)))) false true true true)
                  (Ability|ApplyGameplayEffectToOwner "/Game/Abilities/Player/Generic/GameplayEffects/GE_Generic_Character_Launch.GE_Generic_Character_Launch_C")
                  (Character|BindEventtoOnReachedJumpApex (Variables|Default|GetOwningPawn) (delegate Custom|Redeploy%20Glider))
                  (Utilities|Time|SetTimerbyEvent (delegate Custom|Redeploy%20Glider) 3.5)
                  (Collision|BindEventtoOnActorHit (Variables|Default|GetOwningPawn) (delegate Custom|ActorHit))
                  (Ability|ApplyGameplayEffectToOwner (Variables|Default|GetGE_WilliePete_PlayerLaunch_ApplyGC))
                  (bind _exitingpawn (Variables|/Game/Athena/Items/EnvironmentalItems/HidingProps/Props/B_HidingProp_WilliePete.B_HidingProp_WilliePete_C|GetExitingPawn (Variables|Default|GetWilliePeteLauncher))
                    (:then
                      (if (Utilities|Equal(Object) _exitingpawn (Variables|Default|GetOwningPawn))
                        (Variables|/Game/Athena/Items/EnvironmentalItems/HidingProps/Props/B_HidingProp_WilliePete.B_HidingProp_WilliePete_C|SetExitingPawn 0 (Variables|Default|GetWilliePeteLauncher))))
                    (:else))
                  (elif (Health|IsDBNO (Variables|Default|GetOwningPawn))
                    _
                    (else
                      (CallFunction|StructureTraceLoop)
                      (Ability|Tasks|PlayMontageAndWaitNotify "None" (Variables|Default|GetExitMontage)
                        (:OnCompleted
                          _)
                        (:OnInterrupted
                          _))))))
              (Utilities|FlowControl|Delay 0.5))
            (:Is%20Not%20Valid
              (Ability|EndAbility))))
        (:CastFailed
          _)))
    (:CastFailed
      _))
  (Ability|EndAbility))

(event Custom|ActorHit
  (params (SelfActor :type Actor%20Object%20Reference) (OtherActor :type Actor%20Object%20Reference) (NormalImpulse :type Vector) (Hit :type Hit%20Result%20Structure%20%28by%20ref%29))
  (CallFunction|BreakBlockingStructure OtherActor))

(event Custom|Redeploy%20Glider
  (bind (_returnvalue_x _returnvalue_y _returnvalue_z) (Math|Vector|BreakVector (Transformation|GetActorLocation (Variables|Default|GetOwningPawn))))
  (bind _owningpawn (Variables|Default|GetOwningPawn))
  ;[1 0.2077 0.2376] If not DBNO, Initiate skydive
  (if (not (Health|IsDBNO (Variables|Default|GetOwningPawn)))
    (Collision|UnbindEventfromOnActorHit (Variables|Default|GetOwningPawn) (delegate Custom|ActorHit))
    (Variables|/Game/Athena/Items/EnvironmentalItems/HidingProps/Props/B_HidingProp_WilliePete.B_HidingProp_WilliePete_C|SetskipAnimForLaunch false (Variables|Default|GetWilliePeteLauncher))
    (Parachute|BeginSkydiving (Variables|Default|GetOwningPawn) false)
    (else
      _))
  (EventDispatchers|BindEventtoOnPawnLanded _owningpawn (delegate Custom|End%20Skydive))
  (Character|UnbindallEventsfromOnReachedJumpApex _owningpawn)
  ;;[1 1 1] Remove the GE that granted this ability
  (GameplayEffects|RemoveActiveEffectswithGrantedTags (Ability|GetFortAbilitySystemComponent (Ability|GetAvatarActorfromActorInfo)) (Variables|Default|GetPlayerLaunchTag))
  (Projectile|RemoveGameplayCue%28Looping%29Local (Variables|Default|GetOwningPawn) (Variables|/Game/Athena/Items/EnvironmentalItems/HidingProps/Props/B_HidingProp_WilliePete.B_HidingProp_WilliePete_C|GetGC_ScreenFX (Variables|Default|GetWilliePeteLauncher)) (Ability|GameplayCue|MakeGameplayCueParameters _returnvalue_z))
  (Variables|Default|SetTeleportExitPlayer 0)
  (Ability|EndAbility))

(event Custom|End%20Skydive
  (params (Hit :type Hit%20Result%20Structure%20%28by%20ref%29))
  ;[1 0.2077 0.2376] If not DBNO, Initiate skydive
  (Parachute|EndSkydiving (Variables|Default|GetOwningPawn)))

(fn BreakBlockingStructure
  (params (HitActor :type Actor%20Object%20Reference))
  (if (Variables|/Script/Engine.Actor|GetbCanBeDamaged HitActor)
    (bind _bdidactorhavetags (Class|AthenaQuestFunctionLibrary|AthenaCheckActorHasTags HitActor "(GameplayTags=((TagName=\"Deimos.Spawner\")))"))
    (if (not _bdidactorhavetags)
      (Utilities|Casting|CastToBuildingActor HitActor
        (:then
          _)
        (:CastFailed
          (Utilities|Casting|CastToBuildingProp HitActor
            (:then
              _)
            (:CastFailed
              (Utilities|Casting|CastToAthena_Prop_ParentBuildingContainerBlueprint HitActor)))
          (Utilities|Casting|CastToAthenaBarrierObjective HitActor
            (:CastFailed
              (GameplayEffects|ApplyGameplayEffectToTarget :self (Ability|GetAbilitySystemComponent (Variables|Default|GetOwningPawn)) :GameplayEffectClass "/Game/Athena/Items/Gameplay/Sword/GE_Blade_LeapSlam_BuildingDestroy.GE_Blade_LeapSlam_BuildingDestroy_C" :Target (Ability|GetAbilitySystemComponent HitActor) :Level 0.0))))))))

(fn StructureTraceLoop
  (bind _outhits (Collision|MultiCapsuleTraceForObjects (Transformation|GetWorldLocation (Variables|/Game/Athena/Items/EnvironmentalItems/HidingProps/B_HidingProp.B_HidingProp_C|GetSphere (Variables|Default|GetWilliePeteLauncher))) (Transformation|GetActorLocation (Variables|Default|GetOwningPawn)) (Variables|/Script/Engine.CapsuleComponent|GetCapsuleRadius (Variables|/Script/Engine.Character|GetCapsuleComponent (Variables|Default|GetOwningPawn))) (Variables|/Script/Engine.CapsuleComponent|GetCapsuleHalfHeight (Variables|/Script/Engine.Character|GetCapsuleComponent (Variables|Default|GetOwningPawn))) (Utilities|Array|MakeArray "ObjectTypeQuery1")))
  (bind _array_element (Utilities|Array|ForEachLoopwithBreak _outhits)
    (:LoopBody
      (CallFunction|BreakBlockingStructure (.HitActor (Collision|BreakHitResult _array_element))))))
